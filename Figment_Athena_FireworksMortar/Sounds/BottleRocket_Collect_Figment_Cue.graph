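# Sound cue for collecting a bottle rocket item in the Figment fireworks mortar content.
# A mixer plays three waves at once: real and surreal bottle rocket collect sounds plus a bag loot layer, with the latter two slightly reduced.
# The cue uses prop-collect attenuation and the loot collect sound class at lowered volume.
# A concurrency override caps it at two instances, stopping the oldest.
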SoundCue /Figment_Athena_FireworksMortar/Sounds/BottleRocket_Collect_Figment_Cue
root=Mixer_1
nodes [
 {id=Mixer_1, children [WavePlayer_6, WavePlayer_49, WavePlayer_8], properties {inputVolume [1,0.75,0.7]}}
 {id=WavePlayer_49, properties {soundWaveAssetPtr=/Game/Sounds/Foley_Loot/Ingredients/Munitions/BottleRocket_Collect_Surreal}}
 {id=WavePlayer_6, properties {soundWaveAssetPtr=/Game/Sounds/Foley_Loot/Ingredients/Munitions/BottleRocket_Collect_Real}}
 {id=WavePlayer_8, properties {soundWaveAssetPtr=/Game/Sounds/Foley_Loot/Ingredients/Bag/loot_setup_bagonly_01}}
]
properties {
 attenuationSettings=/Game/Sounds/Attenuation/Fort_Prop_Collect_Att
 bOverrideConcurrency=true
 maxDistance=2097152
 soundClassObject=&Fort_Loot_Collect
 volumeMultiplier=0.451
 concurrencyOverrides {maxCount=2, resolutionRule=StopOldest}
}
